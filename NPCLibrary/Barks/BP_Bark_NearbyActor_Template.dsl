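; Template Blueprint for NPC barks that trigger a speech bubble when tagged player pawns come nearby.
; On BeginPlay, only the authority registers a listener with the Fort nearby-actors perception using a configuration variable.
; Each entering actor is cast to an Athena player pawn and matched against a tag container.
; A match calls PlaySpeechBubble with that pawn and a speech tag; exiting actors are ignored.
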(event EventBeginPlay
  (|CallParentFunction)
  (Networking|SwitchHasAuthority
    (:Authority
      (Query|BPRegisterListener (Variables|Default|GetFortNearbyActorsPerception) (Variables|Default|GetConfiguration) (EventDispatchers|CreateEvent :!event Custom|NewAINearby)))))

(event Custom|NewAINearby
  (params (AllFoundActors :type Array%20of%20Actor%20Object%20References) (EnteringActors :type Array%20of%20Actor%20Object%20References) (ExitingActors :type Array%20of%20Actor%20Object%20References))
  (for _array_element EnteringActors
    (bind _asfort_player_pawn_athena (Utilities|Casting|CastToFortPlayerPawnAthena _array_element))
    (if (GameplayTags|HasAnyMatchingGameplayTags _asfort_player_pawn_athena (Variables|Default|GetTag%20Container))
      (Class|BPFLAICommonFunctions|PlaySpeechBubble _asfort_player_pawn_athena (Variables|Default|GetTag%20for%20Speech)))))

(event Collision|EventActorBeginOverlap
  (params (OtherActor :type Actor%20Object%20Reference)))

(event EventTick
  (params (DeltaSeconds :type Float%20%28single-precision%29)))

(fn ConstructionScript)
